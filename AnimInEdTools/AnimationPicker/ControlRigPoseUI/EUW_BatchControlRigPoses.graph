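# Editor utility widget that saves numbered Control Rig pose assets for one named control, one per click.
# On construct it takes the first Control Rig in the current Level Sequence.
# Save Pose creates an asset named prefix plus count, sets the integer control to that count without keying, selects it and captures selected controls.
# It then increments the count and saves the asset; Reset Count returns the counter to one.
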EditorUtilityWidgetBlueprint /AnimInEdTools/AnimationPicker/ControlRigPoseUI/EUW_BatchControlRigPoses
graphs [EventGraph {nodes [
   "On Clicked (ResetCount)"() {is=title K2Node_ComponentBoundEvent}
   > n1 = Set PoseCountInt(PoseCountInt:int=1) -> (Output_Get:int=0)
   > Set Text(Text="To Text (Integer)"(Value=@n1), self=Get PoseCount() -> (PoseCount:Text Object Reference))
   "On Clicked (SavePose)"() {is=title K2Node_ComponentBoundEvent}
   > n2 = GetAssetTools()
   n3 = Get PosesPathTextBox() -> (PosesPathTextBox:Editor Utility Editable Text Box Object Reference)
   n4 = Get Text(self=Get PosePrefixTextBox() -> (PosePrefixTextBox:Editor Utility Editable Text Box Object Reference))
   n5 = Get PoseCountInt() -> (PoseCountInt:int=0)
   > n6 = CreateAsset(
    self=@n2
    AssetName="To String (Text)"(InText=Format Text!(Format="LOCTEXT(\"{prefix}{int}\")", prefix:Text=@n4, int:int=@n5) -> (Result))
    PackagePath="To String (Text)"(InText=Get Text(self=@n3))
    AssetClass=/Script/ControlRig.ControlRigPoseAsset
   )
   > n7 = Cast To ControlRigPoseAsset!(execute, Object=@n6) -> (then, CastFailed, AsControl Rig Pose Asset:Control Rig Pose Asset Object Reference) {is=title K2Node_DynamicCast}
   > Branch(Condition=IsValid(Object=Get ControlRig() -> (ControlRig:Control Rig Object Reference)))
   n8 = Get ControlTextBox() -> (ControlTextBox:Editor Utility Editable Text Box Object Reference)
   n9 = String To Name(InString="To String (Text)"(InText=Get Text(self=@n8)))
   n10 = Get ControlRig()
   > SetLocalControlRigInt!(
    execute
    LevelSequence:Level Sequence Object Reference=Get LevelSequence() -> (LevelSequence:Level Sequence Object Reference)
    ControlRig:Control Rig Object Reference=@n10
    ControlName:Name=@n9
    Frame_Value:int=0
    Value:int=Get PoseCountInt() -> (PoseCountInt=0)
    TimeUnit:EMovieSceneTimeUnit Enum=DisplayRate
    bSetKey:bool=false
   ) -> (then)
   > SelectControl!(execute, self:Control Rig Object Reference=@n10, InControlName:"Name (by ref)"=@n9, bSelect:bool=true, bSetupUndo:bool=false) -> (then)
   n11 = Get ControlRig()
   > SavePose!(
    execute
    self:Control Rig Pose Asset Object Reference=@n7
    InControlRig:Control Rig Object Reference=@n11
    bUseAll:bool=false
    bSaveValues:bool=true
   ) -> (then)
   > n12 = Set PoseCountInt(PoseCountInt=int + int(A=Get PoseCountInt() -> (PoseCountInt=0))) -> (Output_Get=0)
   > Set Text(Text="To Text (Integer)"(Value=@n12), self=Get PoseCount())
   > SaveLoadedAsset(AssetToSave=@n7, bOnlyIfIsDirty=true)
   Event Construct()
   > n13 = GetCurrentLevelSequence!(execute) -> (then, ReturnValue:Level Sequence Object Reference)
   > n14 = Set LevelSequence(LevelSequence:Level Sequence Object Reference=@n13) -> (Output_Get:Level Sequence Object Reference)
   > n15 = GetControlRigs!(execute, LevelSequence:Level Sequence Object Reference=@n14) -> (then, ReturnValue:Array of Control Rig Sequencer Binding Proxy Structures)
   > Set ControlRig(ControlRig:Control Rig Object Reference="Get (a ref)"!(Array:Array of Control Rig Sequencer Binding Proxy Structures=@n15, Dimension 1=0) -> (
     Output_Proxy:Movie Scene Binding Proxy Structure
     Output_ControlRig:Control Rig Object Reference
     Output_Track:Movie Scene Control Rig Parameter Track Object Reference
    ).Output_ControlRig) -> (Output_Get:Control Rig Object Reference)
   > Set Text(Text="To Text (Integer)"(Value=Get PoseCountInt() -> (PoseCountInt=0)), self=Get PoseCount())
   SelectControl(self=@n11, InControlName=Get TmpControlName() -> (TmpControlName:Name=None), bSelect=true, bSetupUndo=false)
   Event PreConstruct()
   Event Tick()
  ]}]
variables [
 ControlRig {type=ControlRig, default=null}
 LevelSequence {type=LevelSequence, default=null}
 PoseCountInt {type=integer, default=1}
 TmpControlName {type=string, default=arm_ik_l_ctrl}
]
widgets {
 info {rootWidgetClass=&VerticalBox, widgetCount=13}
 widgets [
  VerticalBox_0 {widgetClassPath=&VerticalBox, widget=VerticalBox_0}
  Control {widgetClassPath=&TextBlock, parent=VerticalBox_0, widget=Control, properties {text=Control}, slot {class=VerticalBoxSlot}}
  ControlTextBox {
   widgetClassPath=&EditorUtilityEditableTextBox
   bIsVariable=true
   parent=VerticalBox_0
   widget=ControlTextBox
   properties {text=brow_l_pose}
   slot {class=VerticalBoxSlot}
  }
  TextBlock_190 {
   widgetClassPath=&TextBlock
   parent=VerticalBox_0
   widget=TextBlock_190
   properties {text=Poses Path}
   slot {class=VerticalBoxSlot}
  }
  PosesPathTextBox {
   widgetClassPath=&EditorUtilityEditableTextBox
   bIsVariable=true
   parent=VerticalBox_0
   widget=PosesPathTextBox
   properties {text=/JunoAnimInEdAssets/Figure/Rig/Poses/Face/EyeBrows/Left/}
   slot {class=VerticalBoxSlot}
  }
  PosePrefix {widgetClassPath=&TextBlock, parent=VerticalBox_0, widget=PosePrefix, properties {text=PosePrefix}, slot {class=VerticalBoxSlot}}
  PosePrefixTextBox {
   widgetClassPath=&EditorUtilityEditableTextBox
   bIsVariable=true
   parent=VerticalBox_0
   widget=PosePrefixTextBox
   properties {text=Brow_L_}
   slot {class=VerticalBoxSlot}
  }
  PoseCountText {widgetClassPath=&TextBlock, parent=VerticalBox_0, widget=PoseCountText, properties {text=PoseCount}, slot {class=VerticalBoxSlot}}
  PoseCount {
   widgetClassPath=&TextBlock
   bIsVariable=true
   parent=VerticalBox_0
   widget=PoseCount
   properties {text=0}
   slot {class=VerticalBoxSlot}
  }
  ResetCount {widgetClassPath=&EditorUtilityButton, parent=VerticalBox_0, widget=ResetCount, slot {class=VerticalBoxSlot}}
  TextBlock_98 {widgetClassPath=&TextBlock, parent=ResetCount, widget=TextBlock_98, properties {text=Reset Count}, slot {class=ButtonSlot}}
  SavePose {widgetClassPath=&EditorUtilityButton, parent=VerticalBox_0, widget=SavePose, slot {class=VerticalBoxSlot}}
  TextBlock_50 {widgetClassPath=&TextBlock, parent=SavePose, widget=TextBlock_50, properties {text=Save Pose}, slot {class=ButtonSlot}}
 ]
}
