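# Editor Utility Widget for a labeled asset thumbnail button in the Animation Picker's Control Rig pose UI.
# On Construct, it applies Text to label and tooltips, then AssetData and ThumbnailResolution to the thumbnail.
# A transparent Border inside the button catches left, right and double clicks, firing dispatchers with AssetData; the button's OnClicked also fires the click dispatcher.
# Left click syncs the Content Browser, double-click opens the asset editor, and the right-click handler is empty.
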EditorUtilityWidgetBlueprint /AnimInEdTools/AnimationPicker/ControlRigPoseUI/Thumbnail/EUW_ThumbnailButton
graphs [
 EventGraph {nodes [
   "On Clicked (ThumbnailButton)"() {is=title K2Node_ComponentBoundEvent}
   > Call On Clicked Button Event Dispatcher!(execute:Exec, self:EUW Thumbnail Button Object Reference, AssetData:Asset Data Structure=Get AssetData()) -> (then:Exec) {is=title K2Node_CallDelegate}
   Event Construct()
   > n1 = Set ToolTipText!(
    execute
    ToolTipText:Text=Get Text!() -> (Text)
    self:Widget Object Reference=Get ThumbnailButton() -> (ThumbnailButton:Editor Utility Button Object Reference)
   ) -> (then, Output_Get:Text)
   n2 = Get ThumbnailButtonText() -> (ThumbnailButtonText:Text Object Reference)
   > n3 = Set Text(Text=@n1, self=@n2)
   > Set ToolTipText(ToolTipText=@n3, self=@n2)
   n4 = Get ThumbnailButtonAssetThumbnailWidget() -> (ThumbnailButtonAssetThumbnailWidget:Asset Thumbnail Widget Object Reference)
   > SetAsset!(execute, self:Asset Thumbnail Widget Object Reference=@n4, AssetData:"Asset Data Structure (by ref)"=Get AssetData()) -> (then)
   n5 = Get ThumbnailResolution() -> (ThumbnailResolution:int=0)
   > Set!(execute, Resolution_X:int=@n5, Resolution_Y:int=@n5, self:Asset Thumbnail Widget Object Reference=@n4) -> (then, Output_Get:Int Point Structure)
   n6 = Get!(self:Asset Thumbnail Widget Object Reference=@n4) -> (
    ThumbnailSettings_bForceGenericThumbnail:bool=~
    ThumbnailSettings_bAllowHintText:bool=~
    ThumbnailSettings_bAllowRealTimeOnHovered:bool=~
    ThumbnailSettings_bAllowAssetSpecificThumbnailOverlay:bool=~
    ThumbnailSettings_ThumbnailLabel:EThumbnailLabelType_BlueprintType Enum
    ThumbnailSettings_HighlightedTextDelegate:Delegate
    ThumbnailSettings_HintColorAndOpacity:Linear Color Structure
    ThumbnailSettings_bOverrideAssetTypeColor:bool=~
    ThumbnailSettings_AssetTypeColorOverride:Linear Color Structure
    ThumbnailSettings_Padding:Margin Structure
    ThumbnailSettings_GenericThumbnailSize:int
    ThumbnailSettings_ColorStripOrientation:EThumbnailColorStripOrientation_BlueprintType Enum
   )
   > SetThumbnailSettings!(
    execute
    self:Asset Thumbnail Widget Object Reference=@n4
    InThumbnailSettings_bForceGenericThumbnail:bool=@n6
    InThumbnailSettings_bAllowHintText:bool=@n6.ThumbnailSettings_bAllowHintText
    InThumbnailSettings_bAllowRealTimeOnHovered:bool=@n6.ThumbnailSettings_bAllowRealTimeOnHovered
    InThumbnailSettings_bAllowAssetSpecificThumbnailOverlay:bool=@n6.ThumbnailSettings_bAllowAssetSpecificThumbnailOverlay
    InThumbnailSettings_ThumbnailLabel:EThumbnailLabelType_BlueprintType Enum=@n6.ThumbnailSettings_ThumbnailLabel
    InThumbnailSettings_HighlightedTextDelegate:Delegate=@n6.ThumbnailSettings_HighlightedTextDelegate
    InThumbnailSettings_HintColorAndOpacity:Linear Color Structure=@n6.ThumbnailSettings_HintColorAndOpacity
    InThumbnailSettings_bOverrideAssetTypeColor:bool=@n6.ThumbnailSettings_bOverrideAssetTypeColor
    InThumbnailSettings_AssetTypeColorOverride:Linear Color Structure=@n6.ThumbnailSettings_AssetTypeColorOverride
    InThumbnailSettings_Padding:Margin Structure=@n6.ThumbnailSettings_Padding
    InThumbnailSettings_GenericThumbnailSize:int=@n5
    InThumbnailSettings_ColorStripOrientation:EThumbnailColorStripOrientation_BlueprintType Enum=@n6.ThumbnailSettings_ColorStripOrientation
   ) -> (then)
  ]}
 OnClickedButtonEventDispatcher {nodes [OnClickedButtonEventDispatcher!() -> (then, AssetData:Asset Data Structure) {is=title K2Node_FunctionEntry}]}
 OnDoubleClickedButtonEventDispatcher {nodes [OnDoubleClickedButtonEventDispatcher!() -> (then, AssetData:Asset Data Structure) {is=title K2Node_FunctionEntry}]}
 OnRightClickedButtonEventDispatcher {nodes [OnRightClickedButtonEventDispatcher!() -> (then, AssetData:Asset Data Structure) {is=title K2Node_FunctionEntry}]}
 On_ThumbnailButtonBorder_MouseButtonDown {nodes [
   n1 = On_ThumbnailButtonBorder_MouseButtonDown!() -> (then, MyGeometry:Geometry Structure, MouseEvent:"Pointer Event Structure (by ref)") {is=title K2Node_FunctionEntry}
   > n2 = Sequence!(execute) -> (then_0, then_1, then_2:Exec)
   > Branch(Condition=Is Mouse Button Down!(Input:"Pointer Event Structure (by ref)"=@n1.MouseEvent, MouseButton:Key Structure=RightMouseButton) -> (ReturnValue:bool))
   n3 = Get AssetData()
   > Call On Right Clicked Button Event Dispatcher!(execute:Exec, self:EUW Thumbnail Button Object Reference, AssetData:Asset Data Structure=@n3) -> (then:Exec) {is=title K2Node_CallDelegate}
   > ThumbnailButtonRightClicked!(execute, self, AssetData:Asset Data Structure=@n3) -> (then)
   Branch(execute=@n2.then_1, Condition=Is Mouse Button Down(Input=@n1.MouseEvent, MouseButton=LeftMouseButton))
   n4 = Get AssetData()
   > Call On Clicked Button Event Dispatcher(AssetData=@n4)
   > ThumbnailButtonClicked!(execute, self, AssetData:Asset Data Structure=@n4) -> (then)
   Return Node!(execute=@n2.then_2, ReturnValue:Event Reply Structure=Make Event Reply!() -> (EventReply:Event Reply Structure) {is=title K2Node_MakeStruct})
  ]}
 On_ThumbnailButtonBorder_MouseDoubleClick {nodes [
   On_ThumbnailButtonBorder_MouseDoubleClick!() -> (then, MyGeometry:Geometry Structure, MouseEvent:"Pointer Event Structure (by ref)") {is=title K2Node_FunctionEntry}
   n1 = Get AssetData()
   > Call On Double Clicked Button Event Dispatcher!(execute:Exec, self:EUW Thumbnail Button Object Reference, AssetData:Asset Data Structure=@n1) -> (then:Exec) {is=title K2Node_CallDelegate}
   > ThumbnailButtonDoubleClicked!(execute, self, AssetData:Asset Data Structure=@n1) -> (then)
   > Return Node(ReturnValue=Make Event Reply())
  ]}
 ThumbnailButtonClicked {nodes [
   n1 = ThumbnailButtonClicked!() -> (then, AssetData:Asset Data Structure) {is=title K2Node_FunctionEntry}
   n2 = ToSoftObjectPath!(InAssetData:"Asset Data Structure (by ref)"=@n1) -> (ReturnValue_PathString:String)
   > SyncBrowserToObjects(AssetPaths=Make Array("[0]":String=Split(SourceString=@n2, InStr=., SearchDir=FromStart)) -> (Array:Array of Strings))
  ]}
 ThumbnailButtonDoubleClicked {nodes [
   n1 = ThumbnailButtonDoubleClicked!() -> (then, AssetData:Asset Data Structure) {is=title K2Node_FunctionEntry}
   > n2 = LoadAsset(AssetPath=ToSoftObjectPath(InAssetData=@n1))
   > OpenEditorForAssets!(
    execute
    self:Asset Editor Subsystem Object Reference=Get AssetEditorSubsystem()
    Assets:Array of Object References=Make Array("[0]":Object Reference=@n2) -> (Array:Array of Object References)
    OpenedMethod:EAssetTypeActivationOpenedMethod Enum=Edit
   ) -> (then, ReturnValue:bool)
  ]}
 ThumbnailButtonRightClicked {nodes [
   ThumbnailButtonRightClicked!() -> (then, AssetData:Asset Data Structure) {is=title K2Node_FunctionEntry}
   > Return Node!(execute)
  ]}
]
variables [
 AssetData {type=AssetData, default {}}
 OnClickedButtonEventDispatcher {default=()}
 OnDoubleClickedButtonEventDispatcher {default=()}
 OnRightClickedButtonEventDispatcher {default=()}
 Text {type=string, default=""}
 ThumbnailResolution {type=integer, default=100}
]
widgets {
 info {rootWidgetClass=&VerticalBox, widgetCount=5}
 widgets [
  VerticalBox_37 {widgetClassPath=&VerticalBox, widget=VerticalBox_37}
  ThumbnailButton {
   widgetClassPath=&EditorUtilityButton
   parent=VerticalBox_37
   widget=ThumbnailButton
   properties {widgetStyle {
     clickedSlateSound {resourceObject=null}
     disabled {
      bIsDynamicallyLoaded=false
      drawAs=RoundedBox
      imageType=NoImage
      mirroring=NoMirror
      resourceName=None
      resourceObject=null
      tiling=NoTile
      imageSize=<0 0>
      margin=<ltrb 0 0 0 0>
      outlineSettings {
       bUseBrushTransparency=false
       roundingType=FixedRadius
       width=1
       color {colorUseRule=UseColor_ColorTable, specifiedColor=<rgba 0 0 0 0>}
       cornerRadii=<4 4 4 4>
      }
      tintColor {colorUseRule=UseColor_Specified, specifiedColor=<rgba 0.0395 0.0395 0.0395 1>}
      uVRegion {bIsValid=false, max=<0 0>, min=<0 0>}
     }
     disabledForeground {colorUseRule=UseColor_ColorTable, specifiedColor=<rgba 0 0 0 0>}
     hovered {
      bIsDynamicallyLoaded=false
      drawAs=RoundedBox
      imageType=NoImage
      mirroring=NoMirror
      resourceName=None
      resourceObject=null
      tiling=NoTile
      imageSize=<32 32>
      margin=<ltrb 0 0 0 0>
      outlineSettings {
       bUseBrushTransparency=false
       roundingType=FixedRadius
       width=1
       color {colorUseRule=UseColor_ColorTable, specifiedColor=<rgba 0 0 0 0>}
       cornerRadii=<4 4 4 4>
      }
      tintColor {colorUseRule=UseColor_Specified, specifiedColor=<rgba 0 0.3674 0.5751 1>}
      uVRegion {bIsValid=false, max=<0 0>, min=<0 0>}
     }
     hoveredForeground {colorUseRule=UseColor_ColorTable, specifiedColor=<rgba 0 0 0 0>}
     hoveredSlateSound {resourceObject=null}
     normal {
      bIsDynamicallyLoaded=false
      drawAs=RoundedBox
      imageType=NoImage
      mirroring=NoMirror
      resourceName=None
      resourceObject=null
      tiling=NoTile
      imageSize=<32 32>
      margin=<ltrb 0 0 0 0>
      outlineSettings {
       bUseBrushTransparency=false
       roundingType=FixedRadius
       width=1
       color {colorUseRule=UseColor_ColorTable, specifiedColor=<rgba 0 0 0 0>}
       cornerRadii=<4 4 4 4>
      }
      tintColor {colorUseRule=UseColor_Specified, specifiedColor=<rgba 0.0395 0.0395 0.0395 1>}
      uVRegion {bIsValid=false, max=<0 0>, min=<0 0>}
     }
     normalForeground {colorUseRule=UseColor_ColorTable, specifiedColor=<rgba 0 0 0 0>}
     normalPadding=<ltrb 0 0 0 0>
     pressed {
      bIsDynamicallyLoaded=false
      drawAs=RoundedBox
      imageType=NoImage
      mirroring=NoMirror
      resourceName=None
      resourceObject=null
      tiling=NoTile
      imageSize=<32 32>
      margin=<ltrb 0 0 0 0>
      outlineSettings {
       bUseBrushTransparency=false
       roundingType=FixedRadius
       width=1
       color {colorUseRule=UseColor_ColorTable, specifiedColor=<rgba 0 0 0 0>}
       cornerRadii=<4 4 4 4>
      }
      tintColor {colorUseRule=UseColor_Specified, specifiedColor=<rgba 0.0284 0.0284 0.0284 1>}
      uVRegion {bIsValid=false, max=<0 0>, min=<0 0>}
     }
     pressedForeground {colorUseRule=UseColor_ColorTable, specifiedColor=<rgba 0 0 0 0>}
     pressedPadding=<ltrb 0 0 0 0>
     pressedSlateSound {resourceObject=null}
    }}
   slot {class=VerticalBoxSlot, properties {padding=<ltrb 0 2 0 0>, size {sizeRule=Fill, value=1}}}
  }
  ThumbnailButtonBorder {
   widgetClassPath=&Border
   bIsVariable=true
   parent=ThumbnailButton
   widget=ThumbnailButtonBorder
   properties {brushColor=<rgba 1 1 1 0>, padding=<ltrb 0 0 0 0>}
   slot {class=ButtonSlot, properties {horizontalAlignment=HAlign_Fill, verticalAlignment=VAlign_Fill, padding=<ltrb 3 3 3 3>}}
  }
  ThumbnailButtonAssetThumbnailWidget {
   widgetClassPath=/Script/UMGEditor.AssetThumbnailWidget
   bIsVariable=true
   parent=ThumbnailButtonBorder
   widget=ThumbnailButtonAssetThumbnailWidget
   properties {
    resolution=<100 100>
    thumbnailSettings {
     bAllowAssetSpecificThumbnailOverlay=false
     bAllowHintText=true
     bAllowRealTimeOnHovered=true
     bForceGenericThumbnail=false
     bOverrideAssetTypeColor=false
     colorStripOrientation=HorizontalBottomEdge
     genericThumbnailSize=128
     highlightedTextDelegate="(null).None"
     thumbnailLabel=ClassName
     assetTypeColorOverride=<rgba 0 0 0 0>
     hintColorAndOpacity=<rgba 0 0 0 0>
     padding=<ltrb 0 0 0 0>
    }
   }
   slot {class=BorderSlot, properties {padding=<ltrb 0 0 0 0>}}
  }
  ThumbnailButtonText {
   widgetClassPath=&TextBlock
   bIsVariable=true
   parent=VerticalBox_37
   widget=ThumbnailButtonText
   properties {font {
     bForceMonospaced=false
     bMaterialIsStencil=false
     fontMaterial=null
     fontObject=&Roboto
     letterSpacing=0
     monospacedWidth=1
     size=9
     skewAmount=0
     typefaceFontName=Bold
     outlineSettings {
      bApplyOutlineToDropShadows=false
      bMiteredCorners=false
      bSeparateFillAlpha=false
      outlineMaterial=null
      outlineSize=0
      outlineColor=<rgba 0 0 0 1>
     }
    }}
   slot {class=VerticalBoxSlot, properties {horizontalAlignment=HAlign_Center, padding=<ltrb 0 3 0 5>}}
  }
 ]
}
