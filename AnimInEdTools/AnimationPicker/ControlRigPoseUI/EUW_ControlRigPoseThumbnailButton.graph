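# Editor Utility thumbnail button, derived from EUW_ThumbnailButton, that applies a Control Rig pose asset when clicked.
# A click casts the asset to ControlRigPoseAsset, selects its controls on the Current Control Rig, pastes the pose with keying, then fires On Clicked Button.
# An invalid rig prints a message instead.
# Double-click and right-click overrides do nothing, and no override runs its parent call.
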EditorUtilityWidgetBlueprint /AnimInEdTools/AnimationPicker/ControlRigPoseUI/EUW_ControlRigPoseThumbnailButton
graphs [
 EventGraph {links [], nodes []}
 ThumbnailButtonClicked {nodes [
   n1 = ThumbnailButtonClicked!() -> (then, AssetData:Asset Data Structure) {is=title K2Node_FunctionEntry}
   > n2 = Cast To ControlRigPoseAsset!(execute, Object=GetAsset(InAssetData=@n1)) -> (then, CastFailed, AsControl Rig Pose Asset:Control Rig Pose Asset Object Reference) {is=title K2Node_DynamicCast}
   n3 = Get!() -> (
    Current Control Rig_Proxy:Movie Scene Binding Proxy Structure
    Current Control Rig_ControlRig:Control Rig Object Reference
    Current Control Rig_Track:Movie Scene Control Rig Parameter Track Object Reference
   )
   > n4 = Branch(Condition=IsValid(Object=@n3.Current Control Rig_ControlRig))
   > SelectControls!(
    execute
    self:Control Rig Pose Asset Object Reference=@n2
    InControlRig:Control Rig Object Reference=@n3.Current Control Rig_ControlRig
    bDoMirror:bool=false
    bClearSelection:bool=true
   ) -> (then)
   > PastePose!(
    execute
    self:Control Rig Pose Asset Object Reference=@n2
    InControlRig:Control Rig Object Reference=@n3.Current Control Rig_ControlRig
    bDoKey:bool=true
    bDoMirror:bool=false
    bDoAdditive:bool=false
   ) -> (then)
   > Call On Clicked Button Event Dispatcher!(execute:Exec, self:EUW Thumbnail Button Object Reference, AssetData:Asset Data Structure) -> (then:Exec) {is=title K2Node_CallDelegate}
   PrintString(execute=@n4.else, InString=The ControlRig provided is not valid.)
   Parent: ThumbnailButtonClicked!(execute, AssetData:Asset Data Structure=@n1) -> (then) {is=title K2Node_CallParentFunction}
  ]}
 ThumbnailButtonDoubleClicked {nodes [
   n1 = ThumbnailButtonDoubleClicked!() -> (then, AssetData:Asset Data Structure) {is=title K2Node_FunctionEntry}
   > Return Node()
   Parent: ThumbnailButtonDoubleClicked!(execute, AssetData:Asset Data Structure=@n1) -> (then) {is=title K2Node_CallParentFunction}
  ]}
 ThumbnailButtonRightClicked {nodes [
   n1 = ThumbnailButtonRightClicked!() -> (then, AssetData:Asset Data Structure) {is=title K2Node_FunctionEntry}
   > Return Node()
   Parent: ThumbnailButtonRightClicked!(execute, AssetData:Asset Data Structure=@n1) -> (then) {is=title K2Node_CallParentFunction}
  ]}
]
variables [Current Control Rig {
  type=ControlRigSequencerBindingProxy
  default {controlRig=null, track=null, proxy {bindingId=#1, sequence=null}}
 }]
widgets {
 info {
  parentClass=/AnimInEdTools/AnimationPicker/ControlRigPoseUI/Thumbnail/EUW_ThumbnailButton.EUW_ThumbnailButton_C
  rootWidgetClass=null
  widgetCount=0
 }
 widgets []
}
