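# Editor utility widget for the animation picker that shows Control Rig pose assets as a scrollable grid of thumbnail buttons.
# On Construct, it queries the Asset Registry for each path in a name-to-path map and sorts results through a name-to-index lookup.
# Each button receives the current rig binding and a label stripped of the map key; grid indices restart per path.
# Button clicks are re-broadcast through ThumbnailButtonClickedDispatcher.
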EditorUtilityWidgetBlueprint /AnimInEdTools/AnimationPicker/ControlRigPoseUI/EUW_ControlRigPoses
graphs [
 EventGraph {nodes [
   n1 = OnClickedButtonDispatcher!() -> (OutputDelegate, then, AssetData:Asset Data Structure) {is=title K2Node_CustomEvent}
   > Call Thumbnail Button Clicked Dispatcher!(execute:Exec, self:EUW Control Rig Poses Object Reference) -> (then:Exec) {is=title K2Node_CallDelegate}
   Event Construct()
   n2 = Get Control Rig Poses Paths() -> (Control Rig Poses Paths:Map of Names to Names)
   > n3 = Keys!(execute, TargetMap:Map of Names to Names=@n2) -> (then, Keys:Array of Names)
   > n4 = For Each Loop(Array:Array of Names=@n3) -> (Array Element:Name)
   > Set AssetCount(AssetCount:int=0) -> (Output_Get:int=0)
   > n5 = GetAssetsByPath!(
    execute
    self:Asset Registry Interface=GetAssetRegistry()
    PackagePath:Name=Find(TargetMap:Map of Names to Names=@n2, Key:"Name (by ref)"=@n4) -> (Value:Name)
    bRecursive:bool=false
    bIncludeOnlyOnDiskAssets:bool=false
   ) -> (then, OutAssetData:Array of Asset Data Structures, ReturnValue:bool)
   > n6 = SortAssetDatasByAssetName!(execute, self, AssetDatas:Array of Asset Data Structures=@n5) -> (then, SortedAssetDatas:Array of Asset Data Structures)
   > n7 = For Each Loop(Array:Array of Asset Data Structures=@n6) -> (Array Element:Asset Data Structure)
   n8 = int + int(A=@n7.Array Index)
   > Set AssetCount(AssetCount=@n8) -> (Output_Get=0)
   n9 = "To String (Name)"(InName=@n4)
   n10 = "To String (Name)"(InName=Break Asset Data(AssetData=@n7).AssetName)
   > n11 = Create Widget!(
    execute:Exec
    WorldContextObject:Object Reference
    Class:User Widget Class Reference=&/EUW_ControlRigPoseThumbnailButton.EUW_ControlRigPoseThumbnailButton_C
    OwningPlayer:Player Controller Object Reference
    Current Control Rig:Control Rig Sequencer Binding Proxy Structure=Get Current Control Rig() -> (Current Control Rig:Control Rig Sequencer Binding Proxy Structure)
    AssetData:Asset Data Structure=@n7
    Text:Text="To Text (String)"(InString=Replace(SourceString=@n10, From=@n9))
    ThumbnailResolution:int=Get ThumbnailResolution() -> (ThumbnailResolution:int=0)
   ) -> (then:Exec, ReturnValue:EUW Control Rig Pose Thumbnail Button Object Reference)
   n12 = int - int(A=@n8)
   n13 = Get GridCollumsNumber() -> (GridCollumsNumber:int=0)
   > AddChildToUniformGrid!(
    execute
    self:Uniform Grid Panel Object Reference=Get AssetThumbnailsUniformGrid() -> (AssetThumbnailsUniformGrid:Uniform Grid Panel Object Reference)
    Content:Widget Object Reference=@n11
    InRow:int=int / int(A=@n12, B=@n13)
    InColumn:int="% (Integer)"(A=@n12, B=@n13) -> (ReturnValue=0)
   ) -> (then, ReturnValue:Uniform Grid Slot Object Reference)
   > Bind Event to On Clicked Button Event Dispatcher(self:EUW Thumbnail Button Object Reference=@n11, Delegate=@n1) {is=title K2Node_AddDelegate}
  ]}
 SortAssetDatasByAssetName {nodes [
   n1 = SortAssetDatasByAssetName!() -> (then, AssetDatas:Array of Asset Data Structures) {is=title K2Node_FunctionEntry}
   > n2 = For Each Loop(Array=@n1)
   n3 = Get SortedNames() -> (SortedNames:Array of Names)
   SortNameArray!(
    execute=@n2.Completed
    TargetArray:Array of Names=@n3
    bStableSort:bool=false
    bLexicalSort:bool=true
    SortOrder:EArraySortOrder Enum=Ascending
   ) -> (then)
   > n4 = For Each Loop(Array:Array of Names=@n3) -> (Array Element:Name)
   Return Node!(
    execute=@n4.Completed
    SortedAssetDatas:Array of Asset Data Structures=Get SortedAssetDatasLocal() -> (SortedAssetDatasLocal:Array of Asset Data Structures)
   )
   n5 = Get SortedNamesAndIndex() -> (SortedNamesAndIndex:Map of Names to Integers)
   > Add(
    TargetArray:Array of Asset Data Structures=Get SortedAssetDatasLocal()
    NewItem:"Asset Data Structure (by ref)"="Get (a ref)"(
     Array:Array of Asset Data Structures=Get AssetDatas() -> (AssetDatas:Array of Asset Data Structures)
     Dimension 1=Find(TargetMap:Map of Names to Integers=@n5, Key=@n4) -> (Value:int)
    ) -> (Output:"Asset Data Structure (by ref)")
   ) {is=title K2Node_CallArrayFunction}
   n6 = Break Asset Data(AssetData=@n2)
   Add!(
    execute=@n2.LoopBody
    TargetMap:Map of Names to Integers=Get SortedNamesAndIndex()
    Key:"Name (by ref)"=@n6.AssetName
    Value:"Integer (by ref)"=@n2.Array Index
   ) -> (then) {is=title K2Node_CallFunction}
   > Add(TargetArray:Array of Names=Get SortedNames(), NewItem:"Name (by ref)"=@n6.AssetName) {is=title K2Node_CallArrayFunction}
  ]}
 ThumbnailButtonClickedDispatcher {nodes [ThumbnailButtonClickedDispatcher() {is=title K2Node_FunctionEntry}]}
]
variables [
 AssetCount {type=integer}
 Control Rig Poses Paths {type="map<string>", default {}}
 Current Control Rig {
  type=ControlRigSequencerBindingProxy
  default {controlRig=null, track=null, proxy {bindingId=#1, sequence=null}}
 }
 GridCollumsNumber {type=integer, default=6}
 ThumbnailButtonClickedDispatcher {default=()}
 ThumbnailResolution {type=integer, default=100}
]
widgets {
 info {rootWidgetClass=&VerticalBox, widgetCount=3}
 widgets [
  VerticalBox_0 {widgetClassPath=&VerticalBox, widget=VerticalBox_0}
  ScrollBox_0 {
   widgetClassPath=&ScrollBox
   parent=VerticalBox_0
   widget=ScrollBox_0
   properties {
    widgetBarStyle {
     thickness=8
     draggedThumbImage {
      bIsDynamicallyLoaded=false
      drawAs=RoundedBox
      imageType=NoImage
      mirroring=NoMirror
      resourceName=None
      resourceObject=null
      tiling=NoTile
      imageSize=<0 0>
      margin=<ltrb 0 0 0 0>
      outlineSettings {
       bUseBrushTransparency=false
       roundingType=FixedRadius
       width=0
       color {colorUseRule=UseColor_Specified, specifiedColor=<rgba 0 0 0 0>}
       cornerRadii=<4 4 4 4>
      }
      tintColor {colorUseRule=UseColor_Specified, specifiedColor=<rgba 0.2159 0.2159 0.2159 1>}
      uVRegion {bIsValid=false, max=<0 0>, min=<0 0>}
     }
     horizontalBackgroundImage {
      bIsDynamicallyLoaded=false
      drawAs=NoDrawType
      imageType=NoImage
      mirroring=NoMirror
      resourceName=None
      resourceObject=null
      tiling=NoTile
      imageSize=<0 0>
      margin=<ltrb 0 0 0 0>
      outlineSettings {
       bUseBrushTransparency=false
       roundingType=HalfHeightRadius
       width=0
       color {colorUseRule=UseColor_Specified, specifiedColor=<rgba 0 0 0 0>}
       cornerRadii=<0 0 0 0>
      }
      tintColor {colorUseRule=UseColor_Specified, specifiedColor=<rgba 1 1 1 1>}
      uVRegion {bIsValid=false, max=<0 0>, min=<0 0>}
     }
     horizontalBottomSlotImage {
      bIsDynamicallyLoaded=false
      drawAs=NoDrawType
      imageType=NoImage
      mirroring=NoMirror
      resourceName=None
      resourceObject=null
      tiling=NoTile
      imageSize=<0 0>
      margin=<ltrb 0 0 0 0>
      outlineSettings {
       bUseBrushTransparency=false
       roundingType=HalfHeightRadius
       width=0
       color {colorUseRule=UseColor_Specified, specifiedColor=<rgba 0 0 0 0>}
       cornerRadii=<0 0 0 0>
      }
      tintColor {colorUseRule=UseColor_Specified, specifiedColor=<rgba 1 1 1 1>}
      uVRegion {bIsValid=false, max=<0 0>, min=<0 0>}
     }
     horizontalTopSlotImage {
      bIsDynamicallyLoaded=false
      drawAs=NoDrawType
      imageType=NoImage
      mirroring=NoMirror
      resourceName=None
      resourceObject=null
      tiling=NoTile
      imageSize=<0 0>
      margin=<ltrb 0 0 0 0>
      outlineSettings {
       bUseBrushTransparency=false
       roundingType=HalfHeightRadius
       width=0
       color {colorUseRule=UseColor_Specified, specifiedColor=<rgba 0 0 0 0>}
       cornerRadii=<0 0 0 0>
      }
      tintColor {colorUseRule=UseColor_Specified, specifiedColor=<rgba 1 1 1 1>}
      uVRegion {bIsValid=false, max=<0 0>, min=<0 0>}
     }
     hoveredThumbImage {
      bIsDynamicallyLoaded=false
      drawAs=RoundedBox
      imageType=NoImage
      mirroring=NoMirror
      resourceName=None
      resourceObject=null
      tiling=NoTile
      imageSize=<0 0>
      margin=<ltrb 0 0 0 0>
      outlineSettings {
       bUseBrushTransparency=false
       roundingType=FixedRadius
       width=0
       color {colorUseRule=UseColor_Specified, specifiedColor=<rgba 0 0 0 0>}
       cornerRadii=<4 4 4 4>
      }
      tintColor {colorUseRule=UseColor_Specified, specifiedColor=<rgba 0.2159 0.2159 0.2159 1>}
      uVRegion {bIsValid=false, max=<0 0>, min=<0 0>}
     }
     normalThumbImage {
      bIsDynamicallyLoaded=false
      drawAs=RoundedBox
      imageType=NoImage
      mirroring=NoMirror
      resourceName=None
      resourceObject=null
      tiling=NoTile
      imageSize=<0 0>
      margin=<ltrb 0 0 0 0>
      outlineSettings {
       bUseBrushTransparency=false
       roundingType=FixedRadius
       width=0
       color {colorUseRule=UseColor_Specified, specifiedColor=<rgba 0 0 0 0>}
       cornerRadii=<4 4 4 4>
      }
      tintColor {colorUseRule=UseColor_Specified, specifiedColor=<rgba 0.0953 0.0953 0.0953 1>}
      uVRegion {bIsValid=false, max=<0 0>, min=<0 0>}
     }
     verticalBackgroundImage {
      bIsDynamicallyLoaded=false
      drawAs=NoDrawType
      imageType=NoImage
      mirroring=NoMirror
      resourceName=None
      resourceObject=null
      tiling=NoTile
      imageSize=<0 0>
      margin=<ltrb 0 0 0 0>
      outlineSettings {
       bUseBrushTransparency=false
       roundingType=HalfHeightRadius
       width=0
       color {colorUseRule=UseColor_Specified, specifiedColor=<rgba 0 0 0 0>}
       cornerRadii=<0 0 0 0>
      }
      tintColor {colorUseRule=UseColor_Specified, specifiedColor=<rgba 1 1 1 1>}
      uVRegion {bIsValid=false, max=<0 0>, min=<0 0>}
     }
     verticalBottomSlotImage {
      bIsDynamicallyLoaded=false
      drawAs=NoDrawType
      imageType=NoImage
      mirroring=NoMirror
      resourceName=None
      resourceObject=null
      tiling=NoTile
      imageSize=<0 0>
      margin=<ltrb 0 0 0 0>
      outlineSettings {
       bUseBrushTransparency=false
       roundingType=HalfHeightRadius
       width=0
       color {colorUseRule=UseColor_Specified, specifiedColor=<rgba 0 0 0 0>}
       cornerRadii=<0 0 0 0>
      }
      tintColor {colorUseRule=UseColor_Specified, specifiedColor=<rgba 1 1 1 1>}
      uVRegion {bIsValid=false, max=<0 0>, min=<0 0>}
     }
     verticalTopSlotImage {
      bIsDynamicallyLoaded=false
      drawAs=NoDrawType
      imageType=NoImage
      mirroring=NoMirror
      resourceName=None
      resourceObject=null
      tiling=NoTile
      imageSize=<0 0>
      margin=<ltrb 0 0 0 0>
      outlineSettings {
       bUseBrushTransparency=false
       roundingType=HalfHeightRadius
       width=0
       color {colorUseRule=UseColor_Specified, specifiedColor=<rgba 0 0 0 0>}
       cornerRadii=<0 0 0 0>
      }
      tintColor {colorUseRule=UseColor_Specified, specifiedColor=<rgba 1 1 1 1>}
      uVRegion {bIsValid=false, max=<0 0>, min=<0 0>}
     }
    }
    widgetStyle {
     barThickness=8
     bottomShadowBrush {
      bIsDynamicallyLoaded=false
      drawAs=Box
      imageType=FullColor
      mirroring=NoMirror
      resourceName=../../../Engine/Content/Slate/Common/ScrollBoxShadowBottom.png
      resourceObject=null
      tiling=NoTile
      imageSize=<16 8>
      margin=<ltrb 0.5 0 0.5 1>
      outlineSettings {
       bUseBrushTransparency=false
       roundingType=HalfHeightRadius
       width=0
       color {colorUseRule=UseColor_Specified, specifiedColor=<rgba 0 0 0 0>}
       cornerRadii=<0 0 0 0>
      }
      tintColor {colorUseRule=UseColor_Specified, specifiedColor=<rgba 1 1 1 1>}
      uVRegion {bIsValid=false, max=<0 0>, min=<0 0>}
     }
     horizontalScrolledContentPadding=<ltrb 0 0 1 0>
     leftShadowBrush {
      bIsDynamicallyLoaded=false
      drawAs=Box
      imageType=FullColor
      mirroring=NoMirror
      resourceName=../../../Engine/Content/Slate/Common/ScrollBoxShadowLeft.png
      resourceObject=null
      tiling=NoTile
      imageSize=<8 16>
      margin=<ltrb 1 0.5 0 0.5>
      outlineSettings {
       bUseBrushTransparency=false
       roundingType=HalfHeightRadius
       width=0
       color {colorUseRule=UseColor_Specified, specifiedColor=<rgba 0 0 0 0>}
       cornerRadii=<0 0 0 0>
      }
      tintColor {colorUseRule=UseColor_Specified, specifiedColor=<rgba 1 1 1 1>}
      uVRegion {bIsValid=false, max=<0 0>, min=<0 0>}
     }
     rightShadowBrush {
      bIsDynamicallyLoaded=false
      drawAs=Box
      imageType=FullColor
      mirroring=NoMirror
      resourceName=../../../Engine/Content/Slate/Common/ScrollBoxShadowRight.png
      resourceObject=null
      tiling=NoTile
      imageSize=<8 16>
      margin=<ltrb 0 0.5 1 0.5>
      outlineSettings {
       bUseBrushTransparency=false
       roundingType=HalfHeightRadius
       width=0
       color {colorUseRule=UseColor_Specified, specifiedColor=<rgba 0 0 0 0>}
       cornerRadii=<0 0 0 0>
      }
      tintColor {colorUseRule=UseColor_Specified, specifiedColor=<rgba 1 1 1 1>}
      uVRegion {bIsValid=false, max=<0 0>, min=<0 0>}
     }
     topShadowBrush {
      bIsDynamicallyLoaded=false
      drawAs=Box
      imageType=FullColor
      mirroring=NoMirror
      resourceName=../../../Engine/Content/Slate/Common/ScrollBoxShadowTop.png
      resourceObject=null
      tiling=NoTile
      imageSize=<16 8>
      margin=<ltrb 0.5 1 0.5 0>
      outlineSettings {
       bUseBrushTransparency=false
       roundingType=HalfHeightRadius
       width=0
       color {colorUseRule=UseColor_Specified, specifiedColor=<rgba 0 0 0 0>}
       cornerRadii=<0 0 0 0>
      }
      tintColor {colorUseRule=UseColor_Specified, specifiedColor=<rgba 1 1 1 1>}
      uVRegion {bIsValid=false, max=<0 0>, min=<0 0>}
     }
     verticalScrolledContentPadding=<ltrb 0 0 0 1>
    }
   }
   slot {class=VerticalBoxSlot, properties {size {sizeRule=Fill, value=1}}}
  }
  AssetThumbnailsUniformGrid {
   widgetClassPath=/Script/UMG.UniformGridPanel
   bIsVariable=true
   parent=ScrollBox_0
   widget=AssetThumbnailsUniformGrid
   slot {class=ScrollBoxSlot, properties {horizontalAlignment=HAlign_Center, size {sizeRule=Fill, value=1}}}
  }
 ]
}
